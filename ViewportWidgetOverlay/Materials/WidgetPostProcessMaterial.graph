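# Post-process material that composites a Slate UI texture over the rendered scene, for the ViewportWidgetOverlay content.
# It lerps from the PostProcessInput0 scene color to the SlateUI texture parameter, using texture alpha scaled by OpacityFromTexture.
# The blended result is multiplied by the alpha of TintColorAndOpacity, so that value scales scene and UI together, and the tint's RGB is unused.
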Material /ViewportWidgetOverlay/Materials/WidgetPostProcessMaterial
graph {
 properties {materialDomain=MD_PostProcess, pixelDepthOffsetMode=PDOM_Legacy, usedShadingModels=Default Lit}
 nodes [
  n1 = TextureSampleParameter2D() {parameterName=SlateUI, texture=/Engine/EngineResources/WhiteSquareTexture}
  n2 = ScalarParameter() {defaultValue=1, parameterName=OpacityFromTexture, sliderMax=1}
  Multiply(
   A=LinearInterpolate(A=SceneTexture() {sceneTextureId=PPI_PostProcessInput0}.Color, B=@n1.RGBA, Alpha=Multiply(A=@n1.A, B=@n2))
   B=VectorParameter() {parameterName=TintColorAndOpacity, defaultValue=<rgba 1 1 1 1>}.A
  )
 ]
}
